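; Advanced Storm Controller Creative device that spawns and starts an advanced storm shield with configured radii, phases, delay and finish behavior.
; SpawnStormShield builds phases from storm beacons, either overriding defaults by phase index or replacing them with an insertion-sorted beacon list, falling back to defaults without beacons.
; Start and bounds radius preview meshes rescale cosmetically on property changes and replication.
; Visibility updates also set the looping manager cue's volume.
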(event EventBeginPlay
  (Visibility|SetVisibilityComponents (Variables|Default|GetVisibleInGameComponent) (Utilities|Array|MakeArray (Variables|Default|GetStorm_Manager_Loop_Cue)) true)
  (|CallParentFunction))

(event EventUpdateVisuals
  (params (bVisible :type Boolean))
  (|CallParentFunction bVisible)
  (Audio|Components|Audio|SetVolumeMultiplier (Variables|Default|GetStorm_Manager_Loop_Cue) (select bVisible 1.0 0.0) :!class /Script/Engine.AudioComponent))

(event EventOnPropertiesChanged
  (CallFunction|UpdateStartRadiusStaticMeshComponent)
  (CallFunction|UpdateBoundsRadiusStaticMeshComponent))

(fn GetOverrideMeshMaterial
  (params (MeshComp :type Static%20Mesh%20Component%20Object%20Reference) (MatIdx :type Integer))
  (if (Utilities|Equal(Object) MeshComp (Variables|Default|GetStartRadiusStaticMeshComponent))
    (return (Variables|Transient|GetStartRadiusMID))
    (elif (Utilities|Equal(Object) MeshComp (Variables|Default|GetBoundsRadiusStaticMeshComponent))
      (return (Variables|Transient|GetBoundsRadiusMID))
      (else
        (return (|CallParentFunction (Variables|GetMeshComp) (Variables|GetMatIdx)))))))

(fn GetVisibilityComponents
  (bind _map (|CallParentFunction))
  (Utilities|FlowControl|Sequence
    (:then_0
      (Utilities|Map|Add _map (Variables|Default|GetStartRadiusStaticMeshComponent) "NoCollision"))
    (:then_1
      (Utilities|Map|Add _map (Variables|Default|GetBoundsRadiusStaticMeshComponent) "NoCollision"))
    (:then_2
      (return _map))))

(fn OnRep_BoundsRadius
  (CallFunction|UpdateBoundsRadiusStaticMeshComponent))

(fn OnRep_StartRadius
  (CallFunction|UpdateStartRadiusStaticMeshComponent))

(fn SpawnStormShield
  (bind _stormshield (|CallParentFunction))
  (Utilities|IsValid _stormshield
    (:Is%20Valid
      (bind _asbp_advanced_storm_shield (Utilities|Casting|CastToBP_AdvancedStormShield _stormshield))
      (Utilities|FlowControl|Sequence
        (:then_0
          (switch int (Variables|ToyProperties|GetBeaconType)
            (:0
              ;[0 0 1] Handle beacon overrides
              (Utilities|FlowControl|Sequence
                (:then_0
                  (LocalVariables|SetStormPhases (Variables|Default|GetDefaultStormPhases)))
                (:then_1
                  (bind _foundactors (Utilities|GetActorsRelevanttoMe "/Game/Creative/Devices/StormBeacon/BP_AdvancedStormBeacon.BP_AdvancedStormBeacon_C")
                    (:FoundRelevantActors
                      (for _array_element _foundactors
                        (bind _asbp_advanced_storm_beacon (Utilities|Casting|CastToBP_AdvancedStormBeacon _array_element))
                        (Utilities|Array|AddUnique (LocalVariables|GetStormBeacons) _asbp_advanced_storm_beacon)))))
                (:then_2
                  (for _array_element_2 (LocalVariables|GetStormBeacons)
                    (if (Utilities|Array|IsValidIndex (LocalVariables|GetStormPhases) (- (Variables|/Game/Creative/Devices/StormBeacon/BP_AdvancedStormBeacon.BP_AdvancedStormBeacon_C|GetPhaseIndex _array_element_2) 1))
                      (Utilities|Array|SetArrayElem (LocalVariables|GetStormPhases) (- (Variables|/Game/Creative/Devices/StormBeacon/BP_AdvancedStormBeacon.BP_AdvancedStormBeacon_C|GetPhaseIndex _array_element_2) 1) (Class|BPAdvancedStormBeacon|GetStormPhaseOverride _array_element_2 (Utilities|Array|Get(acopy) (LocalVariables|GetStormPhases) (- (Variables|/Game/Creative/Devices/StormBeacon/BP_AdvancedStormBeacon.BP_AdvancedStormBeacon_C|GetPhaseIndex _array_element_2) 1)))))))))
            (:1
              ;[0 1 0] Handle beacon replacements
              (Utilities|FlowControl|Sequence
                (:then_0
                  (bind _foundactors_2 (Utilities|GetActorsRelevanttoMe "/Game/Creative/Devices/StormBeacon/BP_AdvancedStormBeacon.BP_AdvancedStormBeacon_C")
                    (:FoundRelevantActors
                      (for _array_element_3 _foundactors_2
                        (bind _asbp_advanced_storm_beacon_2 (Utilities|Casting|CastToBP_AdvancedStormBeacon _array_element_3))
                        ;;[1 0 0] Insert sort into the storm beacons array
                        (if (Utilities|IsArrayNonEmpty (LocalVariables|GetStormBeacons))
                          (bind (_array_element_4 _array_index) (Utilities|Array|ForEachLoopwithBreak (LocalVariables|GetStormBeacons))
                            (:LoopBody
                              (if (> (Variables|/Game/Creative/Devices/StormBeacon/BP_AdvancedStormBeacon.BP_AdvancedStormBeacon_C|GetPhaseIndex _array_element_4) (Variables|/Game/Creative/Devices/StormBeacon/BP_AdvancedStormBeacon.BP_AdvancedStormBeacon_C|GetPhaseIndex _asbp_advanced_storm_beacon_2))
                                (Utilities|Array|Insert (LocalVariables|GetStormBeacons) _asbp_advanced_storm_beacon_2 _array_index)))
                            (:Completed
                              _))
                          (Utilities|Array|AddUnique (LocalVariables|GetStormBeacons) _asbp_advanced_storm_beacon_2)
                          (else
                            (Utilities|Array|AddUnique (LocalVariables|GetStormBeacons) _asbp_advanced_storm_beacon_2)))))))
                (:then_1
                  (for (_array_element_5 _array_index_2) (LocalVariables|GetStormBeacons)
                    (Utilities|Array|SetArrayElem (LocalVariables|GetStormPhases) _array_index_2 (Class|BPAdvancedStormBeacon|GetStormPhaseOverride _array_element_5 (Utilities|Array|Get(acopy) (Variables|Default|GetDefaultStormPhases) (Math|Integer|Min(Integer) _array_index_2 (- (Utilities|Array|Length (Variables|Default|GetDefaultStormPhases)) 1)))) true)))
                (:then_2
                  ;;[1 0 0] If no beacons were found, then use the default storm phases
                  (if (not (Utilities|IsArrayNonEmpty (LocalVariables|GetStormBeacons)))
                    (LocalVariables|SetStormPhases (Variables|Default|GetDefaultStormPhases))))))))
        (:then_1
          (Class|BPAdvancedStormShield|Start _asbp_advanced_storm_shield (Variables|ToyProperties|GetStartRadius) true 0 (LocalVariables|GetStormPhases) (Variables|ToyProperties|GetDelayTime) (Variables|ToyProperties|GetOnFinishBehavior) (Variables|ToyProperties|GetBoundsRadius))
          (return _asbp_advanced_storm_shield))))))

(fn UpdateBoundsRadiusStaticMeshComponent
  (bind (_returnvalue_x _returnvalue_y _returnvalue_z) (Math|Vector|BreakVector (Transformation|GetWorldScale (Variables|Default|GetBoundsRadiusStaticMeshComponent))))
  (Networking|CanExecuteCosmeticEvents
    (:True
      (Transformation|SetWorldScale3D (Variables|Default|GetBoundsRadiusStaticMeshComponent) (Math|Vector|MakeVector (CallFunction|ConvertLocalRadiustoWorldRadius (Variables|ToyProperties|GetBoundsRadius)) (CallFunction|ConvertLocalRadiustoWorldRadius (Variables|ToyProperties|GetBoundsRadius)) _returnvalue_z)))))

(fn UpdateStartRadiusStaticMeshComponent
  (bind (_returnvalue_x _returnvalue_y _returnvalue_z) (Math|Vector|BreakVector (Transformation|GetWorldScale (Variables|Default|GetStartRadiusStaticMeshComponent))))
  (Networking|CanExecuteCosmeticEvents
    (:True
      (Transformation|SetWorldScale3D (Variables|Default|GetStartRadiusStaticMeshComponent) (Math|Vector|MakeVector (CallFunction|ConvertLocalRadiustoWorldRadius (Variables|ToyProperties|GetStartRadius)) (CallFunction|ConvertLocalRadiustoWorldRadius (Variables|ToyProperties|GetStartRadius)) _returnvalue_z)))))

(fn ConstructionScript
  (|CallParentFunction)
  (Utilities|FlowControl|Sequence
    (:then_0
      (Variables|Transient|SetStartRadiusMID (Material|GetMID (Variables|Default|GetStartRadiusStaticMeshComponent))))
    (:then_1
      (Variables|Transient|SetBoundsRadiusMID (Material|GetMID (Variables|Default|GetBoundsRadiusStaticMeshComponent))))))
