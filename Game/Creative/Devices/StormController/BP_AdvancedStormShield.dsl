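; Creative-mode Advanced Storm shield that runs a storm circle through configurable phases of waiting, shrinking and moving.
; Start resolves percentage radii and distances, then timers chain delay, hold and shrink, with the authority advancing phases and optionally destroying the shield after the last.
; End locations not preset are randomized, clipped to bounds by segment-cylinder intersection and cached in the phase array.
; RepNotify handlers update clients' countdown times, visibility and audio.
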(event Custom|Start
  (params (InStartRadius :type Float%20%28double-precision%29) (InLatePhasesMove :type Boolean) (InStartingPhaseIndex :type Integer) (InPhases :type Array%20of%20S%20Advanced%20Storm%20Phase%20Structures) (InDelayTime :type Float%20%28double-precision%29) (InOnFinishBehavior :type Integer) (InBoundsRadius :type Float%20%28double-precision%29))
  (bind _phaseindex (Variables|Transient|GetPhaseIndex))
  (Variables|Transient|SetbLatePhasesMove InLatePhasesMove)
  (bind _output_get (Variables|Transient|SetStartingPhaseIndex InStartingPhaseIndex))
  (bind _output_get_2 (Variables|Transient|SetPhaseIndex _output_get))
  (StormShield|SetCurrentPhase _output_get_2)
  (bind _output_get_3 (Variables|Transient|SetPhases InPhases))
  (StormShield|SetNumPhases (Utilities|Array|Length _output_get_3))
  (Variables|Transient|SetDelayTime InDelayTime)
  (Variables|Transient|SetOnFinishBehavior InOnFinishBehavior)
  (Variables|Transient|SetBoundsRadius InBoundsRadius)
  (Variables|Transient|SetBoundsOrigin (Transformation|GetActorLocation))
  (Phase|SetPhaseEndRadius 0 InStartRadius)
  (for _index (range (Utilities|Array|Length (Variables|Transient|GetPhases)))
    (Utilities|FlowControl|Sequence
      (:then_0
        (if (and (> _index 0) (< (Phase|GetPhaseEndRadius _index) 0.0))
          (Phase|SetPhaseEndRadius _index (* (Phase|GetPhaseEndRadius (- _index 1)) (Phase|GetPhaseEndRadiusPercentage _index)))))
      (:then_1
        (Phase|CalculatePhaseForceDistance _index InStartRadius))))
  (Phase|SetPhaseLevel _phaseindex)
  (StormShield|InitializeRadiusValue (Phase|GetPhaseEndRadius _phaseindex))
  (Transformation|SetActorLocation (Phase|GetPhaseEndLocation (Variables|Transient|GetPhaseIndex)))
  (if (> (Variables|Transient|GetDelayTime) 0.0)
    (Utilities|Time|SetTimerbyEvent (EventDispatchers|CreateEvent :!event OnDelayTimerExpired) (Variables|Transient|GetDelayTime))
    (StormShield|SetGamePhaseStep "StormForming")
    (else
      (CallFunction|OnDelayTimerExpired))))

(event StormShield|EventOnRep_GamePhaseStep
  (|CallParentFunction)
  (Networking|CanExecuteCosmeticEvents
    (:True
      (Utilities|FlowControl|Sequence
        (:then_0
          (Development|SetHiddeninGame (Variables|StormShield|GetRoot) (or (< (Variables|StormShield|GetCurrentLocalRadius) 1e+02) (not (or (Utilities|Enum|Equal(Enum) (Variables|FortCreativeStormShield|GetGamePhaseStep) "StormHolding") (Utilities|Enum|Equal(Enum) (Variables|FortCreativeStormShield|GetGamePhaseStep) "StormShrinking")))) true))
        (:then_1
          (switch Utilities|FlowControl|Switch|SwitchonEAthenaGamePhaseStep (Variables|FortCreativeStormShield|GetGamePhaseStep)
            (:StormHolding
              (Audio|PlaySoundatLocation (Variables|Audio|GetBeginHoldingSound) (Variables|StormShield|GetViewTargetLocation))
              (bind _shieldboundaryaudio (Variables|Audio|GetShieldBoundaryAudio)
                (:then
                  (Audio|Components|Audio|SetSound _shieldboundaryaudio (Variables|Audio|GetShieldBoundaryLoopSound)))
                (:else)))
            (:StormShrinking
              (Audio|PlaySoundatLocation (Variables|Audio|GetBeginShrinkingSound) (Variables|StormShield|GetViewTargetLocation))
              (bind _shieldboundaryaudio_2 (Variables|Audio|GetShieldBoundaryAudio)
                (:then
                  (Audio|Components|Audio|SetSound _shieldboundaryaudio_2 (Variables|Audio|GetShieldBoundaryShrinkingLoopSound)))
                (:else)))))
        (:then_2
          (CallFunction|UpdateTimes))))))

(event Custom|Update%20Times
  (Networking|CanExecuteCosmeticEvents
    (:True
      (Variables|FortCreativeStormShield|SetStartWaitTime (+ (Variables|Transient|GetDelayTime) (Variables|FortCreativeStormShield|GetInstancedServerTime)))
      (Variables|FortCreativeStormShield|SetStartShrinkTime (+ (Phase|GetPhaseTotalWaitTime (Variables|Transient|GetPhaseIndex)) (Variables|FortCreativeStormShield|GetStartWaitTime)))
      (Variables|FortCreativeStormShield|SetFinishShrinkTime (+ (Phase|GetPhaseShrinkTime (Variables|Transient|GetPhaseIndex)) (Variables|FortCreativeStormShield|GetStartShrinkTime))))))

(fn CalculatePhaseForceDistance
  (params (Index :type Integer) (StartRadius :type Float%20%28double-precision%29))
  (bind _phases (Variables|Transient|GetPhases))
  (bind (_output_waittime_9_d7e797634847a3af6330d081a73b4f3d _output_shrinktime_11_920d8bb140c25e1b39cce69746cab8cb _output_endradius_28_fc298b6f485aec661de91b9ab8acda81 _output_endradiuspercentage_3_cd98c461485235f1fa37aea8e2858699 _output_damage_25_f609c9dc4f281ba561eb5c8b652cbe3d _output_bmoves_14_1da1bc714b6a48f21e3e4eabe1d63392 _output_forcedistancemin_22_db22fd94466c510f2acb7cb3694c9690 _output_forcedistancemax_24_b7cd745e4b355347c2f906989571146f _output_forcedistanceminpercentage_33_f185525a4b7deb217d04f990787cb987 _output_forcedistancemaxpercentage_34_34aa9d254bc85e7af6cbaa8f770aef4f _output_endlocation_37_c916e3e84c96c120e2615baada29fee4) (Utilities|Struct|BreakSAdvancedStormPhase (Utilities|Array|Get(acopy) _phases Index)))
  (if (Utilities|Array|IsValidIndex _phases Index)
    (Utilities|Array|SetArrayElem (Variables|Transient|GetPhases) Index (Utilities|Struct|MakeSAdvancedStormPhase _output_waittime_9_d7e797634847a3af6330d081a73b4f3d _output_shrinktime_11_920d8bb140c25e1b39cce69746cab8cb _output_endradius_28_fc298b6f485aec661de91b9ab8acda81 -1.0 _output_damage_25_f609c9dc4f281ba561eb5c8b652cbe3d _output_bmoves_14_1da1bc714b6a48f21e3e4eabe1d63392 (Utility|GetFloatOverride _output_forcedistancemin_22_db22fd94466c510f2acb7cb3694c9690 (* _output_forcedistanceminpercentage_33_f185525a4b7deb217d04f990787cb987 StartRadius)) (Utility|GetFloatOverride _output_forcedistancemax_24_b7cd745e4b355347c2f906989571146f (* _output_forcedistancemaxpercentage_34_34aa9d254bc85e7af6cbaa8f770aef4f StartRadius)) -1.0 -1.0 _output_endlocation_37_c916e3e84c96c120e2615baada29fee4))))

(fn DebugPhasesArray
  (for (_array_element _array_index) (Variables|Transient|GetPhases)
    (bind (_array_element_waittime_9_d7e797634847a3af6330d081a73b4f3d _array_element_shrinktime_11_920d8bb140c25e1b39cce69746cab8cb _array_element_endradius_28_fc298b6f485aec661de91b9ab8acda81 _array_element_endradiuspercentage_3_cd98c461485235f1fa37aea8e2858699 _array_element_damage_25_f609c9dc4f281ba561eb5c8b652cbe3d _array_element_bmoves_14_1da1bc714b6a48f21e3e4eabe1d63392 _array_element_forcedistancemin_22_db22fd94466c510f2acb7cb3694c9690 _array_element_forcedistancemax_24_b7cd745e4b355347c2f906989571146f _array_element_forcedistanceminpercentage_33_f185525a4b7deb217d04f990787cb987 _array_element_forcedistancemaxpercentage_34_34aa9d254bc85e7af6cbaa8f770aef4f _array_element_endlocation_37_c916e3e84c96c120e2615baada29fee4) (Utilities|Struct|BreakSAdvancedStormPhase _array_element))
    (Development|PrintText (Utilities|Text|FormatText "NSLOCTEXT(\"[461C53CA4B5EEDC9C286B2A8370A0906]\", \"1F5D14EE494F28257E0542B5B033F1CB\", \"[{7}] = WaitTime: {0}, ShrinkTime: {1}, EndRadius: {2}, Damage: {3}, Moves: {4}, ForceDistance: ({5}, {6}), EndLocation: ({8}, {9}, {10}, {11})\")" _array_index _array_element_waittime_9_d7e797634847a3af6330d081a73b4f3d _array_element_shrinktime_11_920d8bb140c25e1b39cce69746cab8cb _array_element_endradius_28_fc298b6f485aec661de91b9ab8acda81 _array_element_damage_25_f609c9dc4f281ba561eb5c8b652cbe3d (select _array_element_bmoves_14_1da1bc714b6a48f21e3e4eabe1d63392 "NSLOCTEXT(\"[461C53CA4B5EEDC9C286B2A8370A0906]\", \"3953DC30436C5693DEAA12AF2D7F808D\", \"True\")" "NSLOCTEXT(\"[461C53CA4B5EEDC9C286B2A8370A0906]\", \"DEB60AA0401060A91C7D6C9EF7C545EE\", \"False\")") _array_element_forcedistancemin_22_db22fd94466c510f2acb7cb3694c9690 _array_element_forcedistancemax_24_b7cd745e4b355347c2f906989571146f (Math|Vector4|BreakVector4 _array_element_endlocation_37_c916e3e84c96c120e2615baada29fee4) (.Y (Math|Vector4|BreakVector4 _array_element_endlocation_37_c916e3e84c96c120e2615baada29fee4)) (.Z (Math|Vector4|BreakVector4 _array_element_endlocation_37_c916e3e84c96c120e2615baada29fee4)) (.W (Math|Vector4|BreakVector4 _array_element_endlocation_37_c916e3e84c96c120e2615baada29fee4))))))

(fn GetFloatOverride
  (params (Value :type Float%20%28double-precision%29) (Default :type Float%20%28double-precision%29))
  (return (select (>= Value 0.0) Value Default)))

(fn GetNextNextPhaseIndex
  (return (Math|Integer|Min(Integer) (+ (Variables|Transient|GetPhaseIndex) 2) (- (Utilities|Array|Length (Variables|Transient|GetPhases)) 1))))

(fn GetNextPhaseIndex
  (return (+ (Variables|Transient|GetPhaseIndex) 1)))

(fn GetPhaseEndLocation
  (params (Index :type Integer))
  (bind _index (Variables|GetIndex))
  (bind _phases (Variables|Transient|GetPhases))
  (bind _output (Utilities|Array|Get(acopy) _phases _index))
  (bind (_waittime_9_d7e797634847a3af6330d081a73b4f3d _shrinktime_11_920d8bb140c25e1b39cce69746cab8cb _endradius_28_fc298b6f485aec661de91b9ab8acda81 _endradiuspercentage_3_cd98c461485235f1fa37aea8e2858699 _damage_25_f609c9dc4f281ba561eb5c8b652cbe3d _bmoves_14_1da1bc714b6a48f21e3e4eabe1d63392 _forcedistancemin_22_db22fd94466c510f2acb7cb3694c9690 _forcedistancemax_24_b7cd745e4b355347c2f906989571146f _forcedistanceminpercentage_33_f185525a4b7deb217d04f990787cb987 _forcedistancemaxpercentage_34_34aa9d254bc85e7af6cbaa8f770aef4f _endlocation_37_c916e3e84c96c120e2615baada29fee4) (Utilities|Struct|BreakSAdvancedStormPhase _output))
  (bind (_waittime_9_d7e797634847a3af6330d081a73b4f3d_2 _shrinktime_11_920d8bb140c25e1b39cce69746cab8cb_2 _endradius_28_fc298b6f485aec661de91b9ab8acda81_2 _endradiuspercentage_3_cd98c461485235f1fa37aea8e2858699_2 _damage_25_f609c9dc4f281ba561eb5c8b652cbe3d_2 _bmoves_14_1da1bc714b6a48f21e3e4eabe1d63392_2 _forcedistancemin_22_db22fd94466c510f2acb7cb3694c9690_2 _forcedistancemax_24_b7cd745e4b355347c2f906989571146f_2 _forcedistanceminpercentage_33_f185525a4b7deb217d04f990787cb987_2 _forcedistancemaxpercentage_34_34aa9d254bc85e7af6cbaa8f770aef4f_2) (Utilities|Struct|BreakSAdvancedStormPhase (Utilities|Array|Get(acopy) (Variables|Transient|GetPhases) (Variables|GetIndex))))
  (bind _boundsradius (Variables|Transient|GetBoundsRadius))
  (bind _boundsorigin (Variables|Transient|GetBoundsOrigin))
  (bind _desiredmovelocation (Variables|Transient|GetDesiredMoveLocation))
  (bind _returnvalue (Transformation|GetActorLocation))
  (bind _returnvalue_2 (Maths|SegmentCylinderIntersection _returnvalue _desiredmovelocation _boundsorigin _boundsradius))
  (bind _returnvalue_3 (Utilities|Array|Length _returnvalue_2))
  (bind (_forcedistancemin _forcedistancemax) (Phase|GetPhaseForceDistance (Variables|GetIndex)))
  (bind (_returnvalue_x _returnvalue_y _returnvalue_z) (Math|Vector|BreakVector (Transformation|GetActorLocation)))
  (bind (_returnvalue_x_2 _returnvalue_y_2 _returnvalue_z_2) (Math|Vector|BreakVector (Transformation|GetActorLocation)))
  (bind (_x _y _z _w) (Math|Vector4|BreakVector4 _endlocation_37_c916e3e84c96c120e2615baada29fee4))
  (bind (_returnvalue_x_3 _returnvalue_y_3 _returnvalue_z_3) (Math|Vector|BreakVector (Transformation|GetActorLocation)))
  (bind (_returnvalue_roll _returnvalue_pitch _returnvalue_yaw) (Math|Rotator|BreakRotator (Math|Random|RandomRotator false)))
  (bind (_output_x _output_y _output_z) (Math|Vector|BreakVector (Utilities|Array|Get(aref) _returnvalue_2)))
  (bind (_returnvalue_x_4 _returnvalue_y_4 _returnvalue_z_4) (Math|Vector|BreakVector (select (> (Math|Vector|Distance2DSquared(Vector) (Utilities|Array|Get(aref) _returnvalue_2) (Transformation|GetActorLocation)) (Math|Vector|Distance2DSquared(Vector) (Transformation|GetActorLocation) (Utilities|Array|Get(aref) _returnvalue_2))) (Utilities|Array|Get(aref) _returnvalue_2) (Utilities|Array|Get(aref) _returnvalue_2))))
  (bind (_outendlocation_x _outendlocation_y _outendlocation_z) (Math|Vector4|BreakVector4 (Variables|GetOutEndLocation_X)))
  (if (Utilities|Array|IsValidIndex (Variables|Transient|GetPhases) (Variables|GetIndex))
    (if (>= _w 0.0)
      (return _x _y _returnvalue_z_3)
      (else
        _)))
  (if (== (Variables|GetIndex) 0)
    (Variables|SetOutEndLocation_X (Math|Vector4|MakeVector4 _returnvalue_x_2 _returnvalue_y_2 _returnvalue_z_2 1.0)) ; If no movement, then just use the current location
    (else
      (Variables|Transient|SetDesiredMoveLocation (Math|Vector|MakeVector (.x (+ (Transformation|GetActorLocation) (* (Math|Vector|GetForwardVector (Math|Rotator|MakeRotator 0.0 0.0 _returnvalue_yaw)) (Math|Float|SelectFloat (Math|Random|RandomFloatinRange _forcedistancemin _forcedistancemax) (Math|Random|RandomFloatinRange 0.0 (- (Phase|GetPhaseEndRadius (- (Variables|GetIndex) 1)) (Phase|GetPhaseEndRadius (Variables|GetIndex)))) (Phase|PhaseMoves (Variables|GetIndex)))))) (.y (+ (Transformation|GetActorLocation) (* (Math|Vector|GetForwardVector (Math|Rotator|MakeRotator 0.0 0.0 _returnvalue_yaw)) (Math|Float|SelectFloat (Math|Random|RandomFloatinRange _forcedistancemin _forcedistancemax) (Math|Random|RandomFloatinRange 0.0 (- (Phase|GetPhaseEndRadius (- (Variables|GetIndex) 1)) (Phase|GetPhaseEndRadius (Variables|GetIndex)))) (Phase|PhaseMoves (Variables|GetIndex)))))) _returnvalue_z))
      (if (> _returnvalue_3 0)
        (if (== _returnvalue_3 1)
          (Variables|SetOutEndLocation_X (Math|Vector4|MakeVector4 _output_x _output_y _output_z 1.0)) ; If only one intersection point, then just use the first intersection point
          (else
            (Variables|SetOutEndLocation_X (Math|Vector4|MakeVector4 _returnvalue_x_4 _returnvalue_y_4 _returnvalue_z_4 1.0)) ; If two intersection points, then use the one closest to the current location
          ))
        (else
          (Variables|SetOutEndLocation_X (Math|Vector4|MakeVector4 (.x (Variables|Transient|GetDesiredMoveLocation)) (.y (Variables|Transient|GetDesiredMoveLocation)) (.z (Variables|Transient|GetDesiredMoveLocation)) 1.0)) ; Use the evaluated location
        ))))
  (Utilities|Array|SetArrayElem (Variables|Transient|GetPhases) (Variables|GetIndex) (Utilities|Struct|MakeSAdvancedStormPhase _waittime_9_d7e797634847a3af6330d081a73b4f3d_2 _shrinktime_11_920d8bb140c25e1b39cce69746cab8cb_2 _endradius_28_fc298b6f485aec661de91b9ab8acda81_2 _endradiuspercentage_3_cd98c461485235f1fa37aea8e2858699_2 _damage_25_f609c9dc4f281ba561eb5c8b652cbe3d_2 _bmoves_14_1da1bc714b6a48f21e3e4eabe1d63392_2 _forcedistancemin_22_db22fd94466c510f2acb7cb3694c9690_2 _forcedistancemax_24_b7cd745e4b355347c2f906989571146f_2 _forcedistanceminpercentage_33_f185525a4b7deb217d04f990787cb987_2 _forcedistancemaxpercentage_34_34aa9d254bc85e7af6cbaa8f770aef4f_2 (LocalVariables|GetOutEndLocation)) false)
  (return _outendlocation_x _outendlocation_y _outendlocation_z))

(fn GetPhaseEndRadius
  (params (Index :type Integer))
  (bind _phases (Variables|Transient|GetPhases))
  (bind (_output_waittime_9_d7e797634847a3af6330d081a73b4f3d _output_shrinktime_11_920d8bb140c25e1b39cce69746cab8cb _output_endradius_28_fc298b6f485aec661de91b9ab8acda81) (Utilities|Struct|BreakSAdvancedStormPhase (Utilities|Array|Get(acopy) _phases Index)))
  (if (Utilities|Array|IsValidIndex _phases Index)
    (return _output_endradius_28_fc298b6f485aec661de91b9ab8acda81)
    (else
      (return 0.0))))

(fn GetPhaseEndRadiusPercentage
  (params (Index :type Integer))
  (bind _phases (Variables|Transient|GetPhases))
  (bind (_output_waittime_9_d7e797634847a3af6330d081a73b4f3d _output_shrinktime_11_920d8bb140c25e1b39cce69746cab8cb _output_endradius_28_fc298b6f485aec661de91b9ab8acda81 _output_endradiuspercentage_3_cd98c461485235f1fa37aea8e2858699) (Utilities|Struct|BreakSAdvancedStormPhase (Utilities|Array|Get(acopy) _phases Index)))
  (if (Utilities|Array|IsValidIndex _phases Index)
    (return _output_endradiuspercentage_3_cd98c461485235f1fa37aea8e2858699)
    (else
      (return 0.0))))

(fn GetPhaseForceDistance
  (params (Index :type Integer))
  (bind _phases (Variables|Transient|GetPhases))
  (bind (_output_waittime_9_d7e797634847a3af6330d081a73b4f3d _output_shrinktime_11_920d8bb140c25e1b39cce69746cab8cb _output_endradius_28_fc298b6f485aec661de91b9ab8acda81 _output_endradiuspercentage_3_cd98c461485235f1fa37aea8e2858699 _output_damage_25_f609c9dc4f281ba561eb5c8b652cbe3d _output_bmoves_14_1da1bc714b6a48f21e3e4eabe1d63392 _output_forcedistancemin_22_db22fd94466c510f2acb7cb3694c9690 _output_forcedistancemax_24_b7cd745e4b355347c2f906989571146f) (Utilities|Struct|BreakSAdvancedStormPhase (Utilities|Array|Get(acopy) _phases Index)))
  (if (Utilities|Array|IsValidIndex _phases Index)
    (return _output_forcedistancemin_22_db22fd94466c510f2acb7cb3694c9690 _output_forcedistancemax_24_b7cd745e4b355347c2f906989571146f)
    (else
      (return 0.0 0.0))))

(fn GetPhaseShrinkTime
  (params (Index :type Integer))
  (bind _phases (Variables|Transient|GetPhases))
  (bind (_output_waittime_9_d7e797634847a3af6330d081a73b4f3d _output_shrinktime_11_920d8bb140c25e1b39cce69746cab8cb) (Utilities|Struct|BreakSAdvancedStormPhase (Utilities|Array|Get(acopy) _phases Index)))
  (if (Utilities|Array|IsValidIndex _phases Index)
    (return _output_shrinktime_11_920d8bb140c25e1b39cce69746cab8cb)
    (else
      (return 0.1))))

(fn GetPhaseTotalWaitTime
  (params (Index :type Integer))
  (for _index (range (Variables|Transient|GetStartingPhaseIndex) (+ Index 1))
    (LocalVariables|SetOutTotalWaitTime (+ (Phase|GetPhaseWaitTime _index) (LocalVariables|GetOutTotalWaitTime))))
  (if (> (Variables|GetIndex) 0)
    (for _index_2 (range (Variables|Transient|GetStartingPhaseIndex) (Variables|GetIndex))
      (LocalVariables|SetOutTotalWaitTime (+ (Phase|GetPhaseShrinkTime _index_2) (LocalVariables|GetOutTotalWaitTime)))))
  (return (LocalVariables|GetOutTotalWaitTime)))

(fn GetPhaseWaitTime
  (params (Index :type Integer))
  (bind _phases (Variables|Transient|GetPhases))
  (bind _output_waittime_9_d7e797634847a3af6330d081a73b4f3d (Utilities|Struct|BreakSAdvancedStormPhase (Utilities|Array|Get(acopy) _phases Index)))
  (if (Utilities|Array|IsValidIndex _phases Index)
    (return _output_waittime_9_d7e797634847a3af6330d081a73b4f3d)
    (else
      (return 0.1))))

(fn OnDelayTimerExpired
  (CallFunction|StartWaitTimer))

(fn OnRep_DelayTime
  (CallFunction|UpdateTimes))

(fn OnRep_DestroyTime
  (Networking|CanExecuteCosmeticEvents
    (:True
      (if (> (Variables|Transient|GetDestroyTime) 0)
        (Variables|FortCreativeStormShield|SetFinishDestroyTime (+ (Math|Conversions|ToFloat(Integer) (Variables|Transient|GetDestroyTime)) (Variables|FortCreativeStormShield|GetFinishShrinkTime)))))))

(fn OnRep_GamePhaseStepFlashCount
  (StormShield|OnGamePhaseStepChanged))

(fn OnRep_PhaseIndex
  (CallFunction|UpdateTimes))

(fn OnRep_Phases
  (CallFunction|UpdateTimes))

(fn OnShrinkTimerExpired
  (Networking|SwitchHasAuthority
    (:Authority
      (bind _output_get (Variables|Transient|SetPhaseIndex (Phase|GetNextPhaseIndex)))
      (StormShield|SetCurrentPhase _output_get)
      (Phase|SetPhaseLevel _output_get)
      (CallFunction|TransmitonPhaseEnd)
      (if (Utilities|Array|IsValidIndex (Variables|Transient|GetPhases) (Variables|Transient|GetPhaseIndex))
        (CallFunction|StartWaitTimer)
        (else
          (StormShield|SetGamePhaseStep "StormHolding")
          (if (!= (Variables|Transient|GetOnFinishBehavior) -1)
            (if (== (Variables|Transient|GetOnFinishBehavior) 0)
              (Actor|DestroyActor)
              (else
                (Utilities|Time|SetTimerbyEvent (EventDispatchers|CreateEvent :!event K2_DestroyActor) (Math|Float|Max(Float) (Math|Conversions|ToFloat(Integer) (Variables|Transient|GetOnFinishBehavior)) 0.1))
                (Variables|Transient|SetDestroyTime (Variables|Transient|GetOnFinishBehavior))))))))))

(fn OnWaitTimerExpired
  (LocalVariables|SetNextPhaseIndex (Phase|GetNextPhaseIndex))
  (LocalVariables|SetPhaseShrinkTime (Phase|GetPhaseShrinkTime (Variables|Transient|GetPhaseIndex)))
  (StormShield|SetTargetRadiusSizeChangeOverGrowthRate (Phase|GetPhaseEndRadius (LocalVariables|GetNextPhaseIndex)) (LocalVariables|GetPhaseShrinkTime))
  (StormShield|SetTargetLocationtoMoveTo (Phase|GetPhaseEndLocation (LocalVariables|GetNextPhaseIndex)) (LocalVariables|GetPhaseShrinkTime))
  (StormShield|SetGamePhaseStep "StormShrinking")
  (Utilities|Time|SetTimerbyEvent (EventDispatchers|CreateEvent :!event OnShrinkTimerExpired) (Phase|GetPhaseShrinkTime (Variables|Transient|GetPhaseIndex))))

(fn PhaseMoves
  (params (Index :type Integer))
  (bind _phases (Variables|Transient|GetPhases))
  (bind (_output_waittime_9_d7e797634847a3af6330d081a73b4f3d _output_shrinktime_11_920d8bb140c25e1b39cce69746cab8cb _output_endradius_28_fc298b6f485aec661de91b9ab8acda81 _output_endradiuspercentage_3_cd98c461485235f1fa37aea8e2858699 _output_damage_25_f609c9dc4f281ba561eb5c8b652cbe3d _output_bmoves_14_1da1bc714b6a48f21e3e4eabe1d63392) (Utilities|Struct|BreakSAdvancedStormPhase (Utilities|Array|Get(acopy) _phases Index)))
  (if (Utilities|Array|IsValidIndex _phases Index)
    (return _output_bmoves_14_1da1bc714b6a48f21e3e4eabe1d63392)
    (else
      (return false))))

(fn SetPhaseEndRadius
  (params (Index :type Integer) (EndRadius :type Float%20%28double-precision%29))
  (bind _phases (Variables|Transient|GetPhases))
  (bind (_output_waittime_9_d7e797634847a3af6330d081a73b4f3d _output_shrinktime_11_920d8bb140c25e1b39cce69746cab8cb _output_endradius_28_fc298b6f485aec661de91b9ab8acda81 _output_endradiuspercentage_3_cd98c461485235f1fa37aea8e2858699 _output_damage_25_f609c9dc4f281ba561eb5c8b652cbe3d _output_bmoves_14_1da1bc714b6a48f21e3e4eabe1d63392 _output_forcedistancemin_22_db22fd94466c510f2acb7cb3694c9690 _output_forcedistancemax_24_b7cd745e4b355347c2f906989571146f _output_forcedistanceminpercentage_33_f185525a4b7deb217d04f990787cb987 _output_forcedistancemaxpercentage_34_34aa9d254bc85e7af6cbaa8f770aef4f _output_endlocation_37_c916e3e84c96c120e2615baada29fee4) (Utilities|Struct|BreakSAdvancedStormPhase (Utilities|Array|Get(aref) _phases Index)))
  (if (Utilities|Array|IsValidIndex _phases Index)
    (Utilities|Array|SetArrayElem (Variables|Transient|GetPhases) Index (Utilities|Struct|MakeSAdvancedStormPhase _output_waittime_9_d7e797634847a3af6330d081a73b4f3d _output_shrinktime_11_920d8bb140c25e1b39cce69746cab8cb EndRadius _output_endradiuspercentage_3_cd98c461485235f1fa37aea8e2858699 _output_damage_25_f609c9dc4f281ba561eb5c8b652cbe3d _output_bmoves_14_1da1bc714b6a48f21e3e4eabe1d63392 _output_forcedistancemin_22_db22fd94466c510f2acb7cb3694c9690 _output_forcedistancemax_24_b7cd745e4b355347c2f906989571146f _output_forcedistanceminpercentage_33_f185525a4b7deb217d04f990787cb987 _output_forcedistancemaxpercentage_34_34aa9d254bc85e7af6cbaa8f770aef4f _output_endlocation_37_c916e3e84c96c120e2615baada29fee4))))

(fn StartWaitTimer
  (bind _output_get (LocalVariables|SetWaitTime (Phase|GetPhaseWaitTime (Variables|Transient|GetPhaseIndex))))
  (if (> _output_get 0.0)
    (Utilities|Time|SetTimerbyEvent (EventDispatchers|CreateEvent :!event OnWaitTimerExpired) (LocalVariables|GetWaitTime))
    (else
      (CallFunction|OnWaitTimerExpired)))
  (if (Utilities|Enum|NotEqual(Enum) (Variables|FortCreativeStormShield|GetGamePhaseStep) "StormHolding")
    (StormShield|SetGamePhaseStep "StormHolding")
    (else
      (Variables|Transient|SetGamePhaseStepFlashCount (Math|Byte|Byte+Byte (Variables|Transient|GetGamePhaseStepFlashCount) 1))))
  (LocalVariables|SetNextPhaseIndex (Phase|GetNextPhaseIndex))
  (Variables|FortCreativeStormShield|SetNextTargetRadius (Phase|GetPhaseEndRadius (LocalVariables|GetNextPhaseIndex)))
  (Variables|FortCreativeStormShield|SetNextTargetLocation (Phase|GetPhaseEndLocation (LocalVariables|GetNextPhaseIndex)))
  (Variables|FortCreativeStormShield|SetNextNextTargetLocation (Phase|GetPhaseEndLocation (Phase|GetNextNextPhaseIndex))))

(fn ConstructionScript
  (|CallParentFunction))
