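; Basic storm controller device for Creative that spawns a BP_BasicStormShield and previews its start and end radii.
; SpawnStormShield starts the shield with the device's timing, radius, movement and finish properties, mapping Damage through DamageToLevelTable.
; Property changes and radius OnReps rescale the radius meshes on cosmetic-capable machines.
; End-radius phase visibility depends on it being non-zero and different from the start, while visual updates mute or unmute the loop audio.
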(event EventBeginPlay
  (Visibility|SetVisibilityComponents (Variables|Default|GetVisibleInGameComponent) (Utilities|Array|MakeArray (Variables|Default|GetStorm_Manager_Loop_Cue)) true)
  (Visibility|SetVisibilityComponents (Variables|Default|GetEndRadiusVisibleInGameComponent) (Utilities|Array|MakeArray (Variables|Default|GetEndRadiusStaticMeshComponent)) true)
  (|CallParentFunction))

(event EventUpdateVisuals
  (params (bVisible :type Boolean))
  (|CallParentFunction bVisible)
  (Audio|Components|Audio|SetVolumeMultiplier (Variables|Default|GetStorm_Manager_Loop_Cue) (select bVisible 1.0 0.0) :!class /Script/Engine.AudioComponent))

(event EventOnPropertiesChanged
  (CallFunction|UpdateStartRadiusStaticMeshComponent)
  (CallFunction|UpdateEndRadiusStaticMeshComponent))

(fn GetOverrideMeshMaterial
  (params (MeshComp :type Static%20Mesh%20Component%20Object%20Reference) (MatIdx :type Integer))
  (Utilities|FlowControl|Sequence
    (:then_0
      (if (Utilities|Equal(Object) (Variables|GetMeshComp) (Variables|Default|GetStartRadiusStaticMeshComponent))
        (return (Variables|Transient|GetStartRadiusMID))))
    (:then_1
      (if (Utilities|Equal(Object) (Variables|GetMeshComp) (Variables|Default|GetEndRadiusStaticMeshComponent))
        (return (Variables|Transient|GetEndRadiusMID))))))

(fn GetVisibilityComponents
  (bind _map (|CallParentFunction))
  (Utilities|FlowControl|Sequence
    (:then_0
      (Utilities|Map|Add _map (Variables|Default|GetS_StormController) "QueryOnly"))
    (:then_1
      (Utilities|Map|Add _map (Variables|Default|GetStartRadiusStaticMeshComponent) "NoCollision"))
    (:then_2
      (return _map))))

(fn OnRep_EndRadius
  (CallFunction|UpdateEndRadiusStaticMeshComponent))

(fn OnRep_StartRadius
  (CallFunction|UpdateStartRadiusStaticMeshComponent))

(fn SpawnStormShield
  (bind (_value _returnvalue) (Utilities|Map|Find (Variables|Default|GetDamageToLevelTable) (Variables|ToyProperties|GetDamage)))
  (bind _stormshield (|CallParentFunction))
  (Utilities|IsValid _stormshield
    (:Is%20Valid
      (bind _asbp_basic_storm_shield (Utilities|Casting|CastToBP_BasicStormShield _stormshield))
      (Class|BPBasicStormShield|Start _asbp_basic_storm_shield (Variables|ToyProperties|GetDelayTime) (Variables|ToyProperties|GetWaitTime) (Variables|ToyProperties|GetShrinkTime) (Variables|ToyProperties|GetStartRadius) (Variables|ToyProperties|GetEndRadius) (select _returnvalue _value 0) (Variables|ToyProperties|GetOnFinishBehavior) (Variables|ToyProperties|GetMoveDirection) (Variables|ToyProperties|GetMinMoveDistance) (Variables|ToyProperties|GetMaxMoveDistance) (Variables|ToyProperties|GetMoveDelayTime) (Variables|ToyProperties|GetMoveTime))
      (return _asbp_basic_storm_shield))))

(fn UpdateEndRadiusStaticMeshComponent
  (bind (_returnvalue_x _returnvalue_y _returnvalue_z) (Math|Vector|BreakVector (Transformation|GetWorldScale (Variables|Default|GetEndRadiusStaticMeshComponent))))
  (Utilities|FlowControl|Sequence
    (:then_0
      (Visibility|SetVisiblityDuringPhase (Variables|Default|GetEndRadiusVisibleInGameComponent) (select (and (not (Math|Float|NearlyEqual(Float) (Variables|ToyProperties|GetStartRadius) (Variables|ToyProperties|GetEndRadius) 0.01)) (not (Math|Float|NearlyEqual(Float) (Variables|ToyProperties|GetEndRadius)))) "NewEnumerator1" "NewEnumerator0")))
    (:then_1
      (Networking|CanExecuteCosmeticEvents
        (:True
          (Transformation|SetWorldScale3D (Variables|Default|GetEndRadiusStaticMeshComponent) (Math|Vector|MakeVector (CallFunction|ConvertLocalRadiustoWorldRadius (Variables|ToyProperties|GetEndRadius)) (CallFunction|ConvertLocalRadiustoWorldRadius (Variables|ToyProperties|GetEndRadius)) _returnvalue_z)))))))

(fn UpdateStartRadiusStaticMeshComponent
  (bind (_returnvalue_x _returnvalue_y _returnvalue_z) (Math|Vector|BreakVector (Transformation|GetWorldScale (Variables|Default|GetStartRadiusStaticMeshComponent))))
  (Networking|CanExecuteCosmeticEvents
    (:True
      (Transformation|SetWorldScale3D (Variables|Default|GetStartRadiusStaticMeshComponent) (Math|Vector|MakeVector (CallFunction|ConvertLocalRadiustoWorldRadius (Variables|ToyProperties|GetStartRadius)) (CallFunction|ConvertLocalRadiustoWorldRadius (Variables|ToyProperties|GetStartRadius)) _returnvalue_z)))))

(fn ConstructionScript
  (|CallParentFunction)
  (Utilities|FlowControl|Sequence
    (:then_0
      (Variables|Transient|SetStartRadiusMID (Material|GetMID (Variables|Default|GetStartRadiusStaticMeshComponent))))
    (:then_1
      (Variables|Transient|SetEndRadiusMID (Material|GetMID (Variables|Default|GetEndRadiusStaticMeshComponent))))))
