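; Creative storm controller device that spawns and destroys a storm shield for a minigame, apparently a base with empty property and visual hooks.
; Server authority spawns it on game start if enabled, destroys it on game end, and handles generate/destroy events while the minigame runs.
; Spawning places one yaw-aligned shield, passes it a controller reference and queues storm sickness options on the minigame's mutator, dequeued on destroy or EndPlay.
; It also broadcasts phase-end events.
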(event MinigameStartedDelegate(FortMinigameLogic)
  (Networking|SwitchHasAuthority
    (:Authority
      (if (Math|Conversions|ToBoolean(Integer) (Variables|ToyProperties|GetGenerateStormOnGameStart))
        (CallFunction|SpawnStormShield)))))

(event MinigameEndedDelegate(FortMinigameLogic)
  (Networking|SwitchHasAuthority
    (:Authority
      (CallFunction|DestroyStormShield))))

(event EventOnWorldReady
  (bind _map (CallFunction|GetVisibilityComponents))
  (Visibility|SetVisibilityComponentsWithCollision (Variables|Default|GetVisibleInGameComponent) _map true))

(event EventBeginPlay
  (Visibility|AssignOnVisibilityChanged (Variables|Default|GetVisibleInGameComponent) (EventDispatchers|CreateEvent :!event UpdateVisuals)))

(event EventEndPlay
  (params (EndPlayReason :type EEndPlayReason%20Enum :default "Destroyed"))
  (CallFunction|ClearStormSickness))

(event Custom|GeneratedStorm
  (params (Instigator :type Controller%20Object%20Reference))
  (Networking|SwitchHasAuthority
    (:Authority
      (bind _currentminigame (Variables|/Script/FortniteGame.FortMinigameLogicComponent|GetCurrentMinigame (Variables|Default|GetFortMinigameLogic))
        (:then
          (Minigame|IsInProgress _currentminigame
            (:Yes
              (Variables|Transient|GetStormShield
                (:then)
                (:else
                  (CallFunction|SpawnStormShield))))))
        (:else)))))

(event Custom|DestroyStorm
  (params (Instigator :type Controller%20Object%20Reference))
  (Networking|SwitchHasAuthority
    (:Authority
      (bind _currentminigame (Variables|/Script/FortniteGame.FortMinigameLogicComponent|GetCurrentMinigame (Variables|Default|GetFortMinigameLogic))
        (:then
          (Minigame|IsInProgress _currentminigame
            (:Yes
              (CallFunction|DestroyStormShield))))
        (:else)))))

(event Custom|OnPropertiesChanged)

(event OnAnyPropertyChanged(ToyOptionsComponent)
  (CallFunction|OnPropertiesChanged))

(event OnPropertiesLoaded(ToyOptionsComponent)
  (CallFunction|OnPropertiesChanged))

(event Custom|TransmitOnPhaseEnd
  (GameplayEvent|BroadcastEvent (Variables|UserOptions|Events|GetOn%20Phase%20Ended) 0 (Utilities|Struct|MakeGameplayEventwhenPhaseEnded 0)))

(fn BuildDataRegistryResolverScope
  (return (Variables|Default|GetResolver%20Scopes) 1 true))

(fn ClearStormSickness
  (bind _currentminigame (Variables|/Script/FortniteGame.FortMinigameLogicComponent|GetCurrentMinigame (Variables|Default|GetFortMinigameLogic))
    (:then
      (Utilities|IsValid (CallFunction|GetStormSicknessMutator _currentminigame)
        (:Is%20Valid
          (Class|FortAthenaMutatorStormSicknessControl|DequeueStormSickenessOptions (CallFunction|GetStormSicknessMutator _currentminigame) self true))))
    (:else)))

(fn ConvertLocalRadiusToWorldRadius
  (params (LocalRadius :type Float%20%28double-precision%29))
  (bind (_safefromstormplayereffects _instormplayereffects _safefromstormenemyeffects _instormenemyeffects _mapworldscale _replicatedradius _radiusgrowthdata _movedata _radiustoworldscaleconversion) (Class|GetClassDefaults (Variables|Default|GetStormShieldClass)))
  (return (* (* LocalRadius 1e+02) _radiustoworldscaleconversion)))

(fn DestroyStormShield
  (CallFunction|ClearStormSickness)
  (bind _stormshield (Variables|Transient|GetStormShield)
    (:then
      (Actor|DestroyActor _stormshield)
      (Variables|Transient|SetStormShield 0))
    (:else)))

(fn GetCacheContextOverride
  (return 0))

(fn GetStormSicknessMutator
  (params (Minigame :type Fort%20Minigame%20Object%20Reference))
  (bind _returnvalue (Actor|GetComponentbyClass Minigame "/Script/FortniteGame.FortMutatorListComponent"))
  (Utilities|IsValid Minigame
    (:Is%20Valid
      (Utilities|IsValid _returnvalue
        (:Is%20Valid
          (bind _ascreative_mutator_storm_sickness_control (Utilities|Casting|CastToCreativeMutator_StormSicknessControl (Modifier|GetMutatorbyClass _returnvalue "/Game/Athena/Playlists/Creative/Mutators/CreativeMutator_StormSicknessControl.CreativeMutator_StormSicknessControl_C"))
            (:then
              _)
            (:CastFailed
              _))
          (return _ascreative_mutator_storm_sickness_control))
        (:Is%20Not%20Valid
          (return 0))))
    (:Is%20Not%20Valid
      (return 0))))

(fn GetVisibilityComponents
  (return (Utilities|Map|MakeMap (Variables|Debug|GetStaticMeshComponent) "QueryAndPhysics")))

(fn SetStormSickness
  (bind _currentminigame (Variables|/Script/FortniteGame.FortMinigameLogicComponent|GetCurrentMinigame (Variables|Default|GetFortMinigameLogic))
    (:then
      (Utilities|IsValid (CallFunction|GetStormSicknessMutator _currentminigame)
        (:Is%20Valid
          (Class|FortAthenaMutatorStormSicknessControl|EnqueueStormSickenessOptions (CallFunction|GetStormSicknessMutator _currentminigame) (Utilities|Struct|MakeStormSicknessOptions (Variables|UserOptions|Advanced|GetbStormSickness) true self)))))
    (:else)))

(fn SpawnStormShield
  (bind (_returnvalue_roll _returnvalue_pitch _returnvalue_yaw) (Math|Rotator|BreakRotator (Transformation|GetActorRotation)))
  (CallFunction|SetStormSickness)
  (Variables|Transient|GetStormShield
    (:then
      (return (Variables|Transient|GetStormShield)))
    (:else
      (bind _stormshieldclass (Variables|Default|GetStormShieldClass)
        (:then
          (bind _returnvalue (Game|SpawnActorfromClass _stormshieldclass (Math|Transform|MakeTransform (Transformation|GetActorLocation) (Math|Rotator|MakeRotator 0.0 0.0 _returnvalue_yaw)) "AlwaysSpawn" "SelectDefaultAtRuntime" self 0))
          (bind _output_get (Variables|Transient|SetStormShield _returnvalue))
          (bind _asbp_storm_shield (Utilities|Casting|CastToBP_StormShield _output_get)
            (:then
              (Class|BPStormShield|SetStormControllerReference _asbp_storm_shield self :!class /Game/Creative/Devices/StormController/BP_StormShield.BP_StormShield_C))
            (:CastFailed
              _))
          (return _output_get))
        (:else)))))

(fn UpdateVisuals
  (params (bVisible :type Boolean)))

(fn ConstructionScript
  (Variables|Default|SetResolver%20Scopes (Utilities|Array|MakeArray (SpatialGameplay|GetSpatialGameplayInterfaceFNameforActor self))))
