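; Storm shield actor for the Creative Storm Controller device, handling its looping audio, HUD phase updates and phase-end forwarding.
; On BeginPlay it spawns boundary, holding and holding-tick sounds attached to the shield ring and stores their components.
; Game phase step replication refreshes HUD phase visibility where cosmetic events can run.
; TransmitOnPhaseEnd calls through to the storm controller reference, assigned by a setter, when it is valid.
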(event EventBeginPlay
  (|CallParentFunction)
  (CallFunction|InitClosestPointAudio))

(event StormShield|EventOnRep_GamePhaseStep
  (Networking|CanExecuteCosmeticEvents
    (:True
      (Class|MinigameHUDFunctionLibrary|UpdateGamePhaseVisibility self))))

(event Custom|TransmitOnPhaseEnd
  (bind _stormcontrollerreference (Variables|Default|GetStormControllerReference)
    (:then
      (Class|BPStormController|TransmitonPhaseEnd _stormcontrollerreference))
    (:else)))

(fn InitClosestPointAudio
  (Utilities|FlowControl|Sequence
    (:then_0
      (bind _returnvalue (Audio|SpawnSoundAttached :Sound (Variables|Audio|GetShieldBoundaryLoopSound) :AttachToComponent (Variables|StormShield|GetShieldRing) :AttachPointName "None" :LocationType "KeepRelativeOffset" :bStopWhenAttachedToDestroyed true :VolumeMultiplier 1.0 :PitchMultiplier 1.0 :StartTime 0.0 :bAutoDestroy false))
      (Variables|Audio|SetShieldBoundaryAudio _returnvalue))
    (:then_1
      (bind _returnvalue_2 (Audio|SpawnSoundAttached :Sound (Variables|Audio|GetHoldingLoopSound) :AttachToComponent (Variables|StormShield|GetShieldRing) :AttachPointName "None" :LocationType "KeepRelativeOffset" :bStopWhenAttachedToDestroyed true :VolumeMultiplier 1.0 :PitchMultiplier 1.0 :StartTime 0.0 :bAutoDestroy false))
      (Variables|Audio|SetHoldingAudio _returnvalue_2))
    (:then_2
      (bind _returnvalue_3 (Audio|SpawnSoundAttached :Sound (Variables|Audio|GetHoldingLoopTickSound) :AttachToComponent (Variables|StormShield|GetShieldRing) :AttachPointName "None" :LocationType "KeepRelativeOffset" :bStopWhenAttachedToDestroyed true :VolumeMultiplier 1.0 :PitchMultiplier 1.0 :StartTime 0.0 :bAutoDestroy false))
      (Variables|Audio|SetHoldingTickAudio _returnvalue_3))))

(fn SetStormControllerReference
  (params (StormControllerReference :type BP%20Storm%20Controller%20Object%20Reference))
  (Variables|Default|SetStormControllerReference StormControllerReference))

(fn ConstructionScript)
